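# Looping ambient spot SoundCue for a flock of pigeons, played as close-range bird ambience in the Athena ambient sound class.
# A Branch on the IsInRange bool picks the looping vocal flight wave when true and the idle wave when false, with nothing for an unset parameter.
# Only the first quality-level input is connected.
# It limits concurrency per owner and sends to exterior ambient bus and default reverb.
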SoundCue /Game/Sounds/Fort_Ambient_NonLoop/Birds/SC_Spot_Pigeons_Flocking_Loop
root=QualityLevel_0
nodes [
 {id=Branch_0, children [WavePlayer_0,WavePlayer_3,null], properties {boolParameterName=IsInRange}}
 {id=QualityLevel_0, children [Branch_0,null,null,null]}
 {
  id=WavePlayer_0
  properties {bLooping=true, soundWaveAssetPtr=&/Spot_Pigeons_Flocking_Vocal_Flight_01_Loop}
 }
 {
  id=WavePlayer_3
  properties {bLooping=true, soundWaveAssetPtr=&/Spot_Pigeons_Flocking_Vocal_Idle_01_Loop}
 }
]
properties {
 attenuationSettings=/Game/Sounds/Attenuation/ATT_Birds_Flock_Close
 bOverrideConcurrency=true
 maxDistance=1800
 soundClassObject=&Fort_Ambient_Athena
 volumeMultiplier=0.4
 busSends [{audioBus=null, minSendLevel=0.6, sendLevel=1, soundSourceBus=&AmbientAudio_Exterior_SourceBus}]
 concurrencyOverrides {bLimitToOwner=true, maxCount=2, voiceStealReleaseTime=1}
 soundSubmixSends [{minSendLevel=1, soundSubmix=&Fortnite_Reverb_Default}]
}
