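# Niagara environment effect for an idle bird flock (seagulls by system name, Pigeons by emitter) that takes flight when System.IsInRange is set, presumably from User.PlayerScareDistance.
# A randomized counter flips each bird from waddling, with frozen age, to wind- and vortex-driven flight with height-gated kill collision.
# GPU mesh birds avoid distance fields and switch visibility tags and wing material parameters.
# A renderer-less Audio emitter mirrors the motion, playing a pigeon flocking loop.
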NiagaraSystem /Game/Effects/Fort_Effects/Effects/Environment/Hera/TrolleyHills/NS_Env_Seagulls_IdleToFlight
dependencies {
 usedDataInterfaces [
  &NiagaraDataInterfaceCurve
  &NiagaraDataInterfaceCollisionQuery
  &NiagaraDataInterfaceLandscape
  &NiagaraDataInterfaceAudioPlayer
 ]
 usedDynamicInputs [&Boolean_OrOperation, &MakeVectorFromLinearColorRGB]
 usedModules [
  &self:SystemSpawnScript.SystemScriptSource.SystemScriptGraph.NiagaraNodeAssignment_0.SetVariables_#1
  &SystemState
  &self:SystemSpawnScript.SystemScriptSource.SystemScriptGraph.NiagaraNodeAssignment_1.SetVariables_#2
  &self:SystemSpawnScript.SystemScriptSource.SystemScriptGraph.NiagaraNodeAssignment_2.SetVariables_#3
 ]
}
emitters {
 Audio {
  emitterSpawnScript [
   /SetVariables(Emitter.GlobalRadius:NiagaraFloat=&Multiply_Float(A=300, B=$User.RadiusMult)) {isSetParameters=true, module=SetVariables_#4}
   &InitializeAudioPlayer(Audio Player=DataInterface() {propertyValues {
      Attenuation=null
      Concurrency=null
      MaxPlaysPerTick=10
      SoundToPlay=/Game/Sounds/Fort_Ambient_NonLoop/Birds/SC_Spot_Pigeons_Flocking_Loop
      bAllowLoopingOneShotSounds=false
      bLimitPlaysPerTick=true
      bOnlyActiveDuringGameplay=false
      bStopWhenComponentIsDestroyed=true
      ConfigurationUserParameter {parameter {name=None, typeDefHandle {registeredTypeIndex=1}, varData []}}
      ParameterNames [IsInRange]
     }})
  ]
  emitterUpdateScript [
   &EmitterState(
    Enable Distance Culling=true
    Life Cycle Mode=::NewEnumerator1
    Loop Behavior=::NewEnumerator1
    Loop Duration Mode=::NewEnumerator1
    Max Distance Response=::SleepAndClearParticles
    MaxDistance=10000
    Min Distance Response=::SleepAndLetParticlesFinish
    Module.bApplySpawnCountScaleByDistance=false
    Module.bUseMaxDistance=true
    Module.bUseMinDistance=true
    Reset Age On Awaken=true
    Scalability Mode=::NewEnumerator1
    Scale Spawn Count=true
    UseLoopDelay=false
   )
   &SpawnBurst_Instantaneous(Use Spawn Probability=false, UseLoopCountLimit=true)
  ]
  particleSpawnScript [
   &InitializeParticle(
    Color Maximum=<rgba 0.5 0.5 0.5 1>
    Color Minimum=<rgba 1 1 1 1>
    Color Mode=::NewEnumerator2
    Lifetime Max=6
    Lifetime Min=5
    Lifetime Mode=::NewEnumerator1
    Mass Mode=::NewEnumerator3
    Mesh Scale=&Multiply_Vector(
     A=&RandomRangeVector(Minimum=<1 1 1>, Maximum=<1 1.1 1.2>)
     B=&VectorFromFloat(Value=&RandomRangeFloat(Minimum=0.6, Maximum=1))
    )
    Mesh Scale Mode=::NewEnumerator4
    Sprite Rotation Mode=::NewEnumerator2
    Sprite Size Mode=::NewEnumerator3
    Uniform Sprite Size=15
    UsePositionOffset=false
    Write Material Random=true
   )
   &ShapeLocation(
    Cylinder Radius=$Emitter.GlobalRadius
    MaskSpawnGroup=false
    Non Uniform Scale=$User.TransformScale
    Offset Mode=::NewEnumerator0
    Override Seed=false
    Shape Primitive=::NewEnumerator1
    Surface Only=true
    Surface Only Band Thickness=1
    UseOwnerScale=true
   )
   &InitialMeshOrientation(Mesh Axis To Orient=<0 1 0>, Mesh Orientation Mode=::NewEnumerator2, Use Rotation Vector=false)
   /SetVariables(Particles.RandomWaddlePeriod:NiagaraFloat=&RandomRangeFloat(Minimum=0.8, Maximum=1)) {isSetParameters=true, module=SetVariables_#5}
   /SetVariables() {isSetParameters=true, module=SetVariables_#6}
   /Niagara/Modules/Audio/PlayPersistentAudio(Enable Fade In / Out=true, Fade Out Time=1)
  ]
  particleUpdateScript [
   /SetVariables(Particles.IsFlightHighEnough:NiagaraBool=&SetBoolByFloatComparison(
     A=&Abs_Float(Float:NiagaraFloat=&Subtract_Float(
       A=&MakeFloatFromVector(Vector=Linked() {linkedVariable {name=Particles.Position, type=NiagaraPosition}}, Channel=::NewEnumerator2)
       B=&MakeFloatFromVector(Vector=Linked() {linkedVariable {name=Particles.Initial.Position, type=NiagaraPosition}}, Channel=::NewEnumerator2)
      ))
     B=$User.DisableCollisionHeight
    )) {isSetParameters=true}
   /Niagara/Modules/Update/Utility/IncrementOverTime(
    Enabled:NiagaraBool=$System.IsInRange
    Rate of Change:NiagaraFloat=&RandomRangeFloat(Minimum=0.5, Maximum=2)
   )
   /SetVariables(Particles.IsFlying:NiagaraBool=&SetBoolByFloatComparison(A=$Particles.IncrementOverTime.Counter, B=0.5)) {isSetParameters=true, module=SetVariables_#7}
   &ParticleState(DeltaTime:NiagaraFloat=&MakeCustomFloatFromBool(Bool=$Particles.IsFlying, True Float=$Engine.DeltaTime, False Float=0))
   /SetVariables(Particles.DistanceToCenter:NiagaraFloat=&DistanceBetweenPositions(A=$Particles.Position, B=$Engine.Emitter.SimulationPosition)) {isSetParameters=true, module=SetVariables_#8}
   /Niagara/Modules/Update/Utility/VectorNoiseValue(
    Noise Frequency=25
    Pan Noise=true
    Pan Noise Field:Vector3f=<0 0 0.1>
    Remap Output Value:ENiagara_CurlNoiseRemapType=::NewEnumerator2
    Remap Range Output Max:Vector3f=<1 1 0>
    Remap Range Output Min:Vector3f=<-1 -1 0>
   )
   &LinearForce(Force=&Multiply_VectorByFloat(
     Vector=$Output.VectorNoiseValue.Noise
     Float=&Multiply_Float(
      A=&RandomRangeFloat(Minimum=0, Maximum=1)
      B=&MakeCustomFloatFromBool(Bool=$Particles.IsFlying, True Float=0, False Float=200)
     )
    ))
   &PointAttractionForce(
    Attraction Radius:NiagaraFloat=&Multiply_Float(A=$Emitter.GlobalRadius, B=2)
    Attraction Strength=&MakeCustomFloatFromBool(
     Bool=&SetBoolByFloatComparison(A=$Particles.DistanceToCenter, B=$Emitter.GlobalRadius)
     True Float=&MakeCustomFloatFromBool(Bool=$Particles.IsFlying, True Float=0, False Float=3)
     False Float=0
    )
    Kill Within Radius=false
    Module.Use Falloff=true
    UseAttractorOffset=false
   )
   &WindForce(
    Turbulence Mode:ENiagaraWindTurbulenceMode=::NewEnumerator0
    Wind Speed=$User.WindDirection
    Wind Speed Scale=&MakeCustomFloatFromBool(Bool=$Particles.IsFlying, True Float=&RandomRangeFloat(Minimum=250, Maximum=500), False Float=0)
   )
   &VortexForce(
    Origin Pull Amount:NiagaraFloat=&MakeCustomFloatFromBool(Bool=$Particles.IsFlying, True Float=100, False Float=0)
    UseOriginOffset=false
    Vortex Force Amount=&MakeCustomFloatFromBool(
     Bool=$Particles.IsFlying
     True Float=&Lerp_Float(
      A=500
      B=0
      Alpha=&FloatFromCurve(FloatCurve=DataInterface() {propertyValues {Curve {
          defaultValue=1
          keys [
           {arriveTangent=2.8357, interpMode=RCIM_Cubic, leaveTangent=2.8357, tangentMode=RCTM_User, time=0, value=0}
           {arriveTangent=0.0224, interpMode=RCIM_Cubic, leaveTangent=0.0224, tangentMode=RCTM_User, time=0.2, value=1}
          ]
         }}})
     )
     False Float=0
    )
   )
   &Drag(
    Drag=&MakeCustomFloatFromBool(Bool=$Particles.IsFlying, True Float=1, False Float=1)
    Ignore Mass=false
    Use Linear Drag=true
    Use Rotational Drag=true
   )
   &Collision(
    Age Colliding Particles=false
    Collision Enabled=&Boolean_AndOperation(Bool1=$Particles.IsFlying, Bool2=$Particles.IsFlightHighEnough)
    Enable Rest State=false
    Friction=0
    Kill On Collision=5
    KillOnCollision=true
    Module.EnableMaxCollisionCount=false
    Module.Randomize Collision Normal Vector=false
    Particle Radius Scale=40
    Radius Calculation Type=::NewEnumerator1
   )
   &SolveForcesAndVelocity(Clamp Velocity=false, Limit Acceleration=false)
   &UpdateMeshOrientation(
    Facing Direction=&RotateVectorByQuaternion(
     Vector To Rotate=<0 1 0>
     Quaternion=&MakeQuaternion(
      Quaternion From=0
      Axis=<1 0 0>
      Angle=&Sine(
       Normalized Angle=$Emitter.Age
       Period=$Particles.RandomWaddlePeriod
       Scale=&Lerp_Float(A=0, B=0.1, Alpha=&MakeCustomFloatFromBool(Bool=$Particles.IsFlying, True Float=0, False Float=1))
       Bias=0
      )
      From=?
      To=?
      Angle Type=::NewEnumerator1
      Normalize Vectors=false
     )
    )
    Facing Interpolation=false
    Orientation Method=::NewEnumerator1
    Override Quat=false
    TransformFacing=false
    TransformUp=false
   )
   /SetVariables(Particles.VisibilityTag=&MakeCustomIntFromBool(Bool=$Particles.IsFlying, True INT=1, False INT=0)) {isSetParameters=true, module=SetVariables_#9}
   /SetVariables(Particles.WingSinePhase:NiagaraFloat=&Add_Float(
     A=$Particles.WingSinePhase
     B=&Multiply_Float(
      A=$Engine.DeltaTime
      B=&Lerp_Float(
       A=&RandomRangeFloat(Minimum=4, Maximum=8)
       B=&RandomRangeFloat(Minimum=1, Maximum=2)
       Alpha=&FloatFromCurve(FloatCurve=DataInterface() {propertyValues {Curve {
           defaultValue=0
           keys [
            {arriveTangent=0.4313, interpMode=RCIM_Cubic, leaveTangent=0.4313, tangentMode=RCTM_User, time=0, value=0}
            {arriveTangent=0.0224, interpMode=RCIM_Cubic, leaveTangent=0.0224, tangentMode=RCTM_User, time=0.5, value=1}
           ]
          }}})
      )
     )
    )) {isSetParameters=true, module=SetVariables_#10}
   &DynamicMaterialParameters(
    Index 0 Param 1=&Lerp_Float(A=1, B=0.25, Alpha=&MakeFloatFromInt(Int=$Particles.VisibilityTag))
    Index 0 Param 2=&Lerp_Float(
     A=&RandomRangeFloat(Minimum=0, Maximum=2)
     B=$Particles.WingSinePhase
     Alpha=&MakeFloatFromInt(Int=$Particles.VisibilityTag)
    )
    Index 0 Param 3=0.25
    Index 0 Param 4=0.1
    Index 1 Param 1=&Lerp_Float(A=2, B=1, Alpha=&MakeFloatFromInt(Int=$Particles.VisibilityTag))
    Index 1 Param 2=&Lerp_Float(
     A=&RandomRangeFloat(Minimum=2, Maximum=3)
     B=&RandomRangeFloat(Minimum=8, Maximum=10)
     Alpha=&MakeFloatFromInt(Int=$Particles.VisibilityTag)
    )
    Index 1 Param 3=&Lerp_Float(A=0, B=2, Alpha=&MakeFloatFromInt(Int=$Particles.VisibilityTag))
    Index 1 Param 4=10
    Index 2 Param 1=&Lerp_Float(A=100, B=0.1, Alpha=&MakeFloatFromInt(Int=$Particles.VisibilityTag))
    Index 2 Param 2=&Lerp_Float(A=0, B=1, Alpha=&MakeFloatFromInt(Int=$Particles.VisibilityTag))
    Index 2 Param 3=&Lerp_Float(
     A=&RandomRangeFloat(Minimum=1.5, Maximum=3)
     B=0
     Alpha=&MakeFloatFromInt(Int=$Particles.VisibilityTag)
    )
    Index 2 Param 4=&Lerp_Float(
     A=&RandomRangeFloat(Minimum=3, Maximum=10)
     B=0
     Alpha=&MakeFloatFromInt(Int=$Particles.VisibilityTag)
    )
    Index 3 Param 1=&RandomRangeFloat(Minimum=1.5, Maximum=2)
    Index 3 Param 4=&Lerp_Float(A=1, B=1, Alpha=&MakeFloatFromInt(Int=$Particles.VisibilityTag))
    Module.Index0Float0Write=true
    Module.Index0Float1Write=true
    Module.Index0Float2Write=true
    Module.Index0Float3Write=true
    Module.Index1Float0Write=true
    Module.Index1Float1Write=true
    Module.Index1Float2Write=true
    Module.Index1Float3Write=true
    Module.Index2Float0Write=true
    Module.Index2Float1Write=true
    Module.Index2Float2Write=true
    Module.Index2Float3Write=true
    Module.Index3Float0Write=true
    Module.Index3Float1Write=true
    Module.Index3Float2Write=true
    Module.Index3Float3Write=true
    Write Parameter Index 1=::NewEnumerator1
    Write Parameter Index 2=::NewEnumerator1
    Write Parameter Index 3=::NewEnumerator1
   )
   &ScaleMeshSize(Scale Factor=&VectorFromFloat(Value=&FloatFromCurve(FloatCurve=DataInterface() {propertyValues {Curve {keys [
          {tangentMode=RCTM_User, time=0, value=1}
          {arriveTangent=-0.0405, interpMode=RCIM_Cubic, leaveTangent=-0.0405, tangentMode=RCTM_User, time=0.85, value=1}
          {arriveTangent=-8.0089, interpMode=RCIM_Cubic, leaveTangent=-8.0089, tangentMode=RCTM_User, time=1, value=0}
         ]}}})))
   &ScaleColor(
    Scale RGB=&MakeVector(X=&MakeCustomFloatFromBool(Bool=$Particles.IsFlying, True Float=100, False Float=0), Y=0, Z=0)
    ScaleA=true
    ScaleRGB=true
   ) {enabled=false}
   &ScaleColor(
    Scale RGB=&MakeVector(
     X=&MakeCustomFloatFromBool(Bool=&SetBoolByFloatComparison(A=$Particles.IncrementOverTime.Counter, B=0.5), True Float=100, False Float=0)
     Y=0
     Z=0
    )
    ScaleA=true
    ScaleRGB=true
   ) {enabled=false, module=ScaleColor001}
   &ScaleColor(
    Scale RGB=&MakeVector(X=&MakeCustomFloatFromBool(Bool=$System.IsInRange, True Float=100, False Float=0), Y=0, Z=0)
    ScaleA=true
    ScaleRGB=true
   ) {enabled=false, module=ScaleColor002}
   /Niagara/Modules/Audio/UpdatePersistentAudio(Update Type:ENiagara_AudioParamType=::NewEnumerator5)
   &UpdatePersistentAudio(Bool Value:NiagaraBool=$System.IsInRange, Update Type:ENiagara_AudioParamType=::NewEnumerator0) {module=UpdatePersistentAudio001}
  ]
  rendererClasses []
  renderers []
 }
 Pigeons {
  simTarget=GPUComputeSim
  emitterSpawnScript [/SetVariables(Emitter.GlobalRadius:NiagaraFloat=&Multiply_Float(A=300, B=$User.RadiusMult)) {isSetParameters=true, module=SetVariables_#4}]
  emitterUpdateScript [
   &EmitterState(
    Enable Distance Culling=true
    Life Cycle Mode=::NewEnumerator1
    Loop Behavior=::NewEnumerator1
    Loop Duration Mode=::NewEnumerator1
    Max Distance Response=::SleepAndClearParticles
    MaxDistance=10000
    Min Distance Response=::SleepAndLetParticlesFinish
    Module.bApplySpawnCountScaleByDistance=false
    Module.bUseMaxDistance=true
    Module.bUseMinDistance=true
    Reset Age On Awaken=true
    Scalability Mode=::NewEnumerator1
    Scale Spawn Count=true
    UseLoopDelay=false
   )
   &SpawnBurst_Instantaneous(
    Spawn Count=&MakeIntFromFloat(Float=&Multiply_Float(A=&RandomRangeFloat(Minimum=12, Maximum=15), B=$User.SpawnRateMult))
    Use Spawn Probability=false
    UseLoopCountLimit=true
   )
  ]
  particleSpawnScript [
   &InitializeParticle(
    Color Minimum=<rgba 2 2 2 1>
    Color Mode=::NewEnumerator2
    Lifetime Max=6
    Lifetime Min=5
    Lifetime Mode=::NewEnumerator1
    Mass Mode=::NewEnumerator3
    Mesh Scale=&Multiply_Vector(
     A=&RandomRangeVector(Minimum=<1 1 1>, Maximum=<1 1.1 1.2>)
     B=&VectorFromFloat(Value=&RandomRangeFloat(Minimum=0.6, Maximum=1))
    )
    Mesh Scale Mode=::NewEnumerator4
    Sprite Rotation Mode=::NewEnumerator2
    Sprite Size Mode=::NewEnumerator3
    Uniform Sprite Size=15
    UsePositionOffset=false
    Write Material Random=true
   )
   &ShapeLocation(
    Cylinder Radius=$Emitter.GlobalRadius
    MaskSpawnGroup=false
    Non Uniform Scale=$User.TransformScale
    Offset Mode=::NewEnumerator0
    Override Seed=false
    Shape Primitive=::NewEnumerator1
    Surface Only=true
    Surface Only Band Thickness=1
    UseOwnerScale=true
   )
   &InitialMeshOrientation(Mesh Axis To Orient=<0 1 0>, Mesh Orientation Mode=::NewEnumerator2, Use Rotation Vector=false)
   /SetVariables(Particles.RandomWaddlePeriod:NiagaraFloat=&RandomRangeFloat(Minimum=0.8, Maximum=1)) {isSetParameters=true, module=SetVariables_#5}
   /SetVariables() {isSetParameters=true, module=SetVariables_#6}
  ]
  particleUpdateScript [
   /SetVariables(Particles.IsFlightHighEnough:NiagaraBool=&SetBoolByFloatComparison(
     A=&Abs_Float(Float:NiagaraFloat=&Subtract_Float(
       A=&MakeFloatFromVector(Vector=Linked() {linkedVariable {name=Particles.Position, type=NiagaraPosition}}, Channel=::NewEnumerator2)
       B=&MakeFloatFromVector(Vector=Linked() {linkedVariable {name=Particles.Initial.Position, type=NiagaraPosition}}, Channel=::NewEnumerator2)
      ))
     B=$User.DisableCollisionHeight
    )) {isSetParameters=true}
   &IncrementOverTime(
    Enabled:NiagaraBool=$System.IsInRange
    Rate of Change:NiagaraFloat=&RandomRangeFloat(Minimum=0.25, Maximum=2)
   )
   /SetVariables(Particles.IsFlying:NiagaraBool=&SetBoolByFloatComparison(A=$Particles.IncrementOverTime.Counter, B=0.5)) {isSetParameters=true, module=SetVariables_#7}
   /SetVariables(Particles.IsFlying=true) {enabled=false, isSetParameters=true}
   &ParticleState(DeltaTime:NiagaraFloat=&MakeCustomFloatFromBool(Bool=$Particles.IsFlying, True Float=$Engine.DeltaTime, False Float=0))
   /SetVariables(Particles.DistanceToCenter:NiagaraFloat=&DistanceBetweenPositions(A=$Particles.Position, B=$Engine.Emitter.SimulationPosition)) {isSetParameters=true, module=SetVariables_#8}
   &VectorNoiseValue(
    Noise Frequency=25
    Pan Noise=true
    Pan Noise Field:Vector3f=<0 0 0.1>
    Remap Output Value:ENiagara_CurlNoiseRemapType=::NewEnumerator2
    Remap Range Output Max:Vector3f=<1 1 0>
    Remap Range Output Min:Vector3f=<-1 -1 0>
   )
   &LinearForce(Force=&Multiply_VectorByFloat(
     Vector=$Output.VectorNoiseValue.Noise
     Float=&Multiply_Float(
      A=&RandomRangeFloat(Minimum=0, Maximum=1)
      B=&MakeCustomFloatFromBool(Bool=$Particles.IsFlying, True Float=0, False Float=200)
     )
    ))
   &PointAttractionForce(
    Attraction Radius:NiagaraFloat=&Multiply_Float(A=$Emitter.GlobalRadius, B=2)
    Attraction Strength=&MakeCustomFloatFromBool(
     Bool=&SetBoolByFloatComparison(A=$Particles.DistanceToCenter, B=$Emitter.GlobalRadius)
     True Float=&MakeCustomFloatFromBool(Bool=$Particles.IsFlying, True Float=0, False Float=3)
     False Float=0
    )
    Kill Within Radius=false
    Module.Use Falloff=true
    UseAttractorOffset=false
   )
   &WindForce(
    Frequency=7
    Friction Mode:ENiagaraWindFrictionMode=::NewEnumerator0
    Scale Min / Max:Vector2f=<0 0.4>
    Speed Min / Max:Vector2f=<0 400>
    Turbulence Scale Mode:ENiagaraWindTurbulenceContributionMode=::NewEnumerator3
    Wind Speed=$User.WindDirection
    Wind Speed Scale=&MakeCustomFloatFromBool(Bool=$Particles.IsFlying, True Float=&RandomRangeFloat(Minimum=250, Maximum=500), False Float=0)
   )
   &VortexForce(
    Origin Pull Amount:NiagaraFloat=&MakeCustomFloatFromBool(Bool=$Particles.IsFlying, True Float=100, False Float=0)
    UseOriginOffset=false
    Vortex Force Amount=&MakeCustomFloatFromBool(
     Bool=$Particles.IsFlying
     True Float=&Lerp_Float(
      A=500
      B=0
      Alpha=&FloatFromCurve(FloatCurve=DataInterface() {propertyValues {Curve {
          defaultValue=1
          keys [
           {arriveTangent=2.8357, interpMode=RCIM_Cubic, leaveTangent=2.8357, tangentMode=RCTM_User, time=0, value=0}
           {arriveTangent=0.0224, interpMode=RCIM_Cubic, leaveTangent=0.0224, tangentMode=RCTM_User, time=0.2, value=1}
          ]
         }}})
     )
     False Float=0
    )
   )
   &Drag(
    Drag=&MakeCustomFloatFromBool(Bool=$Particles.IsFlying, True Float=1, False Float=1)
    Ignore Mass=false
    Use Linear Drag=true
    Use Rotational Drag=true
   )
   /Niagara/Modules/Math/AvoidDistanceFieldSurfaces_GPU(
    Active:NiagaraBool=&Boolean_AndOperation(Bool1=$Particles.IsFlying, Bool2=&SetBoolByFloatComparison(A=$Particles.NormalizedAge, B=0.1))
    Avoid Nearest Surface=true
    Avoid Oncoming Surfaces=true
    Nearest Surface Avoidance Strength=2500
   )
   &Collision(
    Age Colliding Particles=false
    Collision Enabled=&Boolean_AndOperation(Bool1=$Particles.IsFlying, Bool2=$Particles.IsFlightHighEnough)
    Enable Rest State=false
    Friction=0
    GPU Collision Type=::NewEnumerator2
    Kill On Collision=5
    KillOnCollision=true
    Module.EnableMaxCollisionCount=false
    Module.Randomize Collision Normal Vector=false
    Particle Radius Scale=40
    Radius Calculation Type=::NewEnumerator1
   )
   &SolveForcesAndVelocity(Clamp Velocity=false, Limit Acceleration=false)
   &UpdateMeshOrientation(
    Facing Direction=&RotateVectorByQuaternion(
     Vector To Rotate=<0 1 0>
     Quaternion=&MakeQuaternion(
      Quaternion From=0
      Axis=<1 0 0>
      Angle=&Sine(
       Normalized Angle=$Emitter.Age
       Period=$Particles.RandomWaddlePeriod
       Scale=&Lerp_Float(A=0, B=0.05, Alpha=&MakeCustomFloatFromBool(Bool=$Particles.IsFlying, True Float=0, False Float=1))
       Bias=0
      )
      From=?
      To=?
      Angle Type=::NewEnumerator1
      Normalize Vectors=false
     )
    )
    Facing Interpolation=false
    Orientation Method=::NewEnumerator1
    Override Quat=false
    TransformFacing=false
    TransformUp=false
   )
   /SetVariables(Particles.VisibilityTag=&MakeCustomIntFromBool(Bool=$Particles.IsFlying, True INT=1, False INT=0)) {isSetParameters=true, module=SetVariables_#9}
   /SetVariables(Particles.WingSinePhase:NiagaraFloat=&Add_Float(
     A=$Particles.WingSinePhase
     B=&Multiply_Float(
      A=$Engine.DeltaTime
      B=&Lerp_Float(
       A=&RandomRangeFloat(Minimum=4, Maximum=8)
       B=&RandomRangeFloat(Minimum=1, Maximum=2)
       Alpha=&FloatFromCurve(FloatCurve=DataInterface() {propertyValues {Curve {
           defaultValue=0
           keys [
            {arriveTangent=0.4313, interpMode=RCIM_Cubic, leaveTangent=0.4313, tangentMode=RCTM_User, time=0, value=0}
            {arriveTangent=0.0224, interpMode=RCIM_Cubic, leaveTangent=0.0224, tangentMode=RCTM_User, time=0.5, value=1}
           ]
          }}})
      )
     )
    )) {isSetParameters=true, module=SetVariables_#10}
   &DynamicMaterialParameters(
    Index 0 Param 1=&Lerp_Float(A=1, B=0.25, Alpha=&MakeFloatFromInt(Int=$Particles.VisibilityTag))
    Index 0 Param 2=&Lerp_Float(
     A=&RandomRangeFloat(Minimum=0, Maximum=2)
     B=$Particles.WingSinePhase
     Alpha=&MakeFloatFromInt(Int=$Particles.VisibilityTag)
    )
    Index 0 Param 3=0.25
    Index 0 Param 4=0.1
    Index 1 Param 1=&Lerp_Float(A=2, B=1, Alpha=&MakeFloatFromInt(Int=$Particles.VisibilityTag))
    Index 1 Param 2=&Lerp_Float(
     A=&RandomRangeFloat(Minimum=2, Maximum=3)
     B=&RandomRangeFloat(Minimum=8, Maximum=10)
     Alpha=&MakeFloatFromInt(Int=$Particles.VisibilityTag)
    )
    Index 1 Param 3=&Lerp_Float(A=0, B=2, Alpha=&MakeFloatFromInt(Int=$Particles.VisibilityTag))
    Index 1 Param 4=10
    Index 2 Param 1=&Lerp_Float(A=100, B=0.1, Alpha=&MakeFloatFromInt(Int=$Particles.VisibilityTag))
    Index 2 Param 2=&Lerp_Float(A=0, B=1, Alpha=&MakeFloatFromInt(Int=$Particles.VisibilityTag))
    Index 2 Param 3=&Lerp_Float(
     A=&RandomRangeFloat(Minimum=1.5, Maximum=3)
     B=0
     Alpha=&MakeFloatFromInt(Int=$Particles.VisibilityTag)
    )
    Index 2 Param 4=&Lerp_Float(
     A=&RandomRangeFloat(Minimum=3, Maximum=10)
     B=0
     Alpha=&MakeFloatFromInt(Int=$Particles.VisibilityTag)
    )
    Index 3 Param 1=&RandomRangeFloat(Minimum=1.5, Maximum=2)
    Index 3 Param 4=&Lerp_Float(A=1, B=1, Alpha=&MakeFloatFromInt(Int=$Particles.VisibilityTag))
    Module.Index0Float0Write=true
    Module.Index0Float1Write=true
    Module.Index0Float2Write=true
    Module.Index0Float3Write=true
    Module.Index1Float0Write=true
    Module.Index1Float1Write=true
    Module.Index1Float2Write=true
    Module.Index1Float3Write=true
    Module.Index2Float0Write=true
    Module.Index2Float1Write=true
    Module.Index2Float2Write=true
    Module.Index2Float3Write=true
    Module.Index3Float0Write=true
    Module.Index3Float1Write=true
    Module.Index3Float2Write=true
    Module.Index3Float3Write=true
    Write Parameter Index 1=::NewEnumerator1
    Write Parameter Index 2=::NewEnumerator1
    Write Parameter Index 3=::NewEnumerator1
   )
   &ScaleMeshSize(Scale Factor=&VectorFromFloat(Value=&FloatFromCurve(FloatCurve=DataInterface() {propertyValues {Curve {keys [
          {tangentMode=RCTM_User, time=0, value=1}
          {arriveTangent=-0.0405, interpMode=RCIM_Cubic, leaveTangent=-0.0405, tangentMode=RCTM_User, time=0.85, value=1}
          {arriveTangent=-8.0089, interpMode=RCIM_Cubic, leaveTangent=-8.0089, tangentMode=RCTM_User, time=1, value=0}
         ]}}})))
   &ScaleColor(
    Scale RGB=&MakeVector(X=&MakeCustomFloatFromBool(Bool=$Particles.IsFlying, True Float=100, False Float=0), Y=0, Z=0)
    ScaleA=true
    ScaleRGB=true
   ) {enabled=false}
   &ScaleColor(
    Scale RGB=&MakeVector(
     X=&MakeCustomFloatFromBool(Bool=&SetBoolByFloatComparison(A=$Particles.IncrementOverTime.Counter, B=0.5), True Float=100, False Float=0)
     Y=0
     Z=0
    )
    ScaleA=true
    ScaleRGB=true
   ) {enabled=false, module=ScaleColor001}
   &ScaleColor(
    Scale RGB=&MakeVector(X=&MakeCustomFloatFromBool(Bool=$System.IsInRange, True Float=100, False Float=0), Y=0, Z=0)
    ScaleA=true
    ScaleRGB=true
   ) {enabled=false, module=ScaleColor002}
   /SetVariables(Particles.DEBUG_Speed:NiagaraFloat=&VectorLength(Vector=$Particles.Velocity)) {enabled=false, isSetParameters=true}
  ]
  rendererClasses [&NiagaraMeshRendererProperties]
  renderers [
   {rendererClass=&NiagaraMeshRendererProperties, rendererIndex=0}
   {rendererClass=&NiagaraMeshRendererProperties, rendererIndex=1}
   {rendererClass=&NiagaraMeshRendererProperties, rendererIndex=2}
  ]
 }
}
summary {
 systemName=NS_Env_Seagulls_IdleToFlight
 emitters [
  {emitterName=Pigeons, simTarget=GPUComputeSim, rendererClasses [&NiagaraMeshRendererProperties]}
  {emitterName=Audio, rendererClasses []}
 ]
 userVariables [
  User.DisableCollisionHeight {type=NiagaraFloat, defaultValue=50, description=""}
  User.PlayerScareDistance {type=NiagaraFloat, defaultValue=3500, description=""}
  User.RadiusMult {type=NiagaraFloat, defaultValue=1, description=""}
  User.SpawnRateMult {type=NiagaraFloat, defaultValue=1, description=""}
  User.TransformScale {type=Vector3f, description="", defaultValue {struct=Vector3f, value=<1 1 0>}}
  User.WindDirection {type=Vector3f, description="", defaultValue {struct=Vector3f, value=<0.5 0 1>}}
 ]
}
